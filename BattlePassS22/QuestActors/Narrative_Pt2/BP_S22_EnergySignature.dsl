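; Season 22 Energy Signature quest prop that registers and unregisters itself as a tracked quest actor for each player.
; Pre-interaction visuals start tracking, while hidden and post-interaction visuals and state changes stop it, through gameplay events sent to the player's pawn with this actor as payload.
; Quest callbacks set visuals, interactability and distance watching from objective state or a visual-test flag.
; On full quest completion, authority removes a configured gameplay effect and optionally resets cosmetic overrides.
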(event EventBeginPlay
  (|CallParentFunction) ; This node is disabled and will not be called.\nDrag off pins to build functionality.
)

(event Collision|EventActorBeginOverlap
  (params (OtherActor :type Actor%20Object%20Reference))
  (|CallParentFunction OtherActor) ; This node is disabled and will not be called.\nDrag off pins to build functionality.
)

(event EventTick
  (params (DeltaSeconds :type Float%20%28single-precision%29))
  (|CallParentFunction DeltaSeconds) ; This node is disabled and will not be called.\nDrag off pins to build functionality.
)

(event EventVisualTestingNoQuestRequired_Init
  (|CallParentFunction)
  (CallFunction|StartTracking (Fortnite|GetFirstLocalFortPlayerController true)))

(event EventResetforVisualTesting
  (params (PlayerController :type Fort%20Player%20Controller%20Object%20Reference))
  (|CallParentFunction PlayerController)
  (if (Variables|Design|QuestSetup|GetbTestVisualsNoQuestRequired)
    (CallFunction|StartTracking PlayerController)))

(event OnQuestReady(ConsolidatedQuestComponent)
  (params (PlayerController :type Fort%20Player%20Controller%20Object%20Reference))
  (if (or (ObjectiveState|IsObjectiveActive PlayerController) (Variables|Design|QuestSetup|GetbTestVisualsNoQuestRequired))
    (if (Variables|Advanced|IntervalCheck|GetbUsingDistanceCheck)
      (CallFunction|EventAddWatchedActor PlayerController)
      (else
        _))
    (if (not (Variables|Design|GetbUseDiscoverCollisionToUnhide))
      (InteractionState|HandleSetCanInteract PlayerController)
      (VisualState|Meshes|PreInteractionVisuals PlayerController)
      (VisualState|VFX&Audio|PreInteractionVFX&Audio PlayerController))
    (elif (Variables|Design|QuestSetup|GetbVisibleToAll)
      (VisualState|Meshes|PostInteractionVisuals PlayerController)
      (else
        (VisualState|Meshes|HiddenVisuals PlayerController)))))

(event OnQuestComponentSuccessEvent(ConsolidatedQuestComponent)
  (params (PlayerController :type Fort%20Player%20Controller%20Object%20Reference))
  (bind _consolidatedquestcomponent (Variables|BuildingGameplayActorPropQuest|GetConsolidatedQuestComponent))
  (bind _returnvalue (Class|ConsolidatedQuestComponent|GetQuestItemDefinition _consolidatedquestcomponent))
  (bind _returnvalue_2 (Quests|GetQuestManager PlayerController "Invalid"))
  (bind _returnvalue_3 (Quests|GetQuestwithDefinition _returnvalue_2 _returnvalue))
  (CallFunction|UpdateOverlappingActors)
  (CallFunction|UpdateVisuals PlayerController)
  (Networking|SwitchHasAuthority
    (:Authority
      (Utilities|IsValid _returnvalue_3
        (:Is%20Valid
          (if (Quest|HasCompletedAllObjectives _returnvalue_3 false)
            (CallFunction|RemoveGEOnQuestCompletion PlayerController)
            (if (and (Variables|Design|QuestSetup|GameplayTags,EffectsandCosmetics|GetbShouldApplyCosmeticOverrides) (Variables|Design|QuestSetup|GameplayTags,EffectsandCosmetics|GetbResetOverridesOnCompletion))
              (CallFunction|CosmeticOverridestoPlayer PlayerController true)
              (Development|PrintString "REMOVED" (Variables|Design|QuestSetup|GetbTestVisualsNoQuestRequired) (Variables|Design|QuestSetup|GetbTestVisualsNoQuestRequired) "(R=0.000000,G=1.000000,B=0.073006,A=1.000000)"))
            (else
              (Development|PrintString (Utilities|GetDisplayName _returnvalue)))))))))

(event OnUpdateQuest(ConsolidatedQuestComponent)
  (params (PlayerController :type Fort%20Player%20Controller%20Object%20Reference))
  (if (or (ObjectiveState|IsObjectiveActive PlayerController) (Variables|Design|QuestSetup|GetbTestVisualsNoQuestRequired))
    (VisualState|Meshes|PreInteractionVisuals PlayerController)
    (if (Variables|Advanced|IntervalCheck|GetbUsingDistanceCheck)
      (CallFunction|EventAddWatchedActor PlayerController)
      (else
        _))
    (if (not (Variables|Design|GetbUseDiscoverCollisionToUnhide))
      (InteractionState|HandleSetCanInteract PlayerController))))

(event EventChangeStateForPlayer
  (params (FPC :type Fort%20Player%20Controller%20Object%20Reference))
  (|CallParentFunction FPC)
  (CallFunction|StopTracking FPC))

(fn HiddenVisuals
  (params (PlayerController :type Fort%20Player%20Controller%20Object%20Reference))
  (|CallParentFunction PlayerController)
  (CallFunction|StopTracking PlayerController))

(fn PostInteractionVisuals
  (params (PlayerController :type Fort%20Player%20Controller%20Object%20Reference))
  (|CallParentFunction PlayerController)
  (CallFunction|StopTracking PlayerController))

(fn PreInteractionVisuals
  (params (PlayerController :type Fort%20Player%20Controller%20Object%20Reference))
  (|CallParentFunction PlayerController)
  (CallFunction|StartTracking PlayerController))

(fn RemoveGEOnQuestCompletion
  (params (FPC :type Fort%20Player%20Controller%20Object%20Reference))
  (GameplayEffects|RemoveActiveGameplayEffectbySourceEffect (Ability|GetAbilitySystemComponent FPC) (Variables|Default|GetGERemovedOnCompletion) 0 -1))

(fn StartTracking
  (params (FPC :type Fort%20Player%20Controller%20Object%20Reference))
  (Ability|SendGameplayEventtoActor (Game|Player|GetPlayerPawn FPC) "(TagName=\"Event.Quest.TrackQuestActor.Add.S22.EnergySignature\")" (Utilities|Struct|MakeGameplayEventData "(TagName=\"\")" 0 self)))

(fn StopTracking
  (params (FPC :type Fort%20Player%20Controller%20Object%20Reference))
  (Ability|SendGameplayEventtoActor (Game|Player|GetPlayerPawn FPC) "(TagName=\"Event.Quest.TrackQuestActor.Remove.S22.EnergySignature\")" (Utilities|Struct|MakeGameplayEventData "(TagName=\"\")" 0 self)))

(fn ConstructionScript
  (|CallParentFunction))
